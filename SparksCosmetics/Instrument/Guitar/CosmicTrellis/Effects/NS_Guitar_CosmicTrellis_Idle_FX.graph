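# Idle-effect Niagara system for the Cosmic Trellis guitar cosmetic, with no active emitters in its current state.
# Its only emitter, Minimal, is a disabled sprite template with an instantaneous burst and particles that are not killed when their lifetime elapses.
# System scripts use two Set Variables modules, and the system references color-curve and float-from-vector inputs without showing their targets.
# It exposes IsInstrument and bIsFrontEndPreview user parameters, which the emitter does not visibly read.
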NiagaraSystem /SparksCosmetics/Instrument/Guitar/CosmicTrellis/Effects/NS_Guitar_CosmicTrellis_Idle_FX
dependencies {
 usedDataInterfaces [&NiagaraDataInterfaceColorCurve]
 usedDynamicInputs [&ColorFromCurve, &MakeFloatFromVector]
 usedModules [
  &SystemState
  &self:SystemSpawnScript.SystemScriptSource.SystemScriptGraph.NiagaraNodeAssignment_1.SetVariables_#1
  &self:SystemSpawnScript.SystemScriptSource.SystemScriptGraph.NiagaraNodeAssignment_2.SetVariables_#2
 ]
}
emitters {Minimal {
  bEnabled=false
  emitterUpdateScript [
   &EmitterState()
   &SpawnBurst_Instantaneous(Use Spawn Probability=false, UseLoopCountLimit=true)
  ]
  particleSpawnScript [&InitializeParticle(
    Color Mode=::NewEnumerator0
    Lifetime=5
    Sprite Size Mode=::NewEnumerator3
    Uniform Sprite Size=50
    UsePositionOffset=false
    Write Material Random=true
   )]
  particleUpdateScript [&ParticleState(Kill Particles When Lifetime Has Elapsed=false)]
  rendererClasses [&NiagaraSpriteRendererProperties]
  renderers [{rendererClass=&NiagaraSpriteRendererProperties}]
 }}
summary {
 systemName=NS_Guitar_CosmicTrellis_Idle_FX
 emitters [{bEnabled=false, emitterName=Minimal, rendererClasses [&NiagaraSpriteRendererProperties]}]
 userVariables [
  User.bIsFrontEndPreview {type=NiagaraBool}
  User.IsInstrument {type=NiagaraFloat, defaultValue=0, description=""}
 ]
}
