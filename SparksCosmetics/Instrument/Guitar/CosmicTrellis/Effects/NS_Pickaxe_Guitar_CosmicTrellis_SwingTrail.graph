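# Niagara swing trail for the Cosmic Trellis guitar pickaxe cosmetic, drawn as two anim-trail ribbons from bone or socket positions.
# TrailSharp is a short-lived yellow streak with width shaped over particle life; TrailSoft is a longer-lived, dim amber trail.
# Both tint by a system-level ToDColor and fade alpha through emitter-age curves with a night boost from FNTimeOfDayScalar.
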NiagaraSystem /SparksCosmetics/Instrument/Guitar/CosmicTrellis/Effects/NS_Pickaxe_Guitar_CosmicTrellis_SwingTrail
dependencies {
 usedDataInterfaces [&NiagaraDataInterfaceCurve, &NiagaraDataInterfaceSocketReader, &NiagaraDataInterfaceColorCurve]
 usedDynamicInputs [&ColorFromCurve]
 usedModules [
  &self:SystemSpawnScript.SystemScriptSource.SystemScriptGraph.NiagaraNodeAssignment_4.SetVariables_#1
  &SystemState
  &self:SystemSpawnScript.SystemScriptSource.SystemScriptGraph.NiagaraNodeAssignment_5.SetVariables_#2
 ]
}
emitters {
 TrailSharp {
  emitterSpawnScript [&FN_RibbonIDOffsetUtility()]
  emitterUpdateScript [
   &EmitterState()
   &FN_SpawnAnimTrail()
  ]
  particleSpawnScript [
   &FN_InitializeAnimTrailBoneIndecies(First Bone or Socket Index=1)
   &FN_InitializeAnimTrail(AnimTrailWidth=3, Color=<rgba 1 0.8139 0 1>, Lifetime=0.18)
   &AddVelocity(Velocity=<800 0 0>) {enabled=false}
  ]
  particleUpdateScript [
   &SolveForcesAndVelocity(Clamp Velocity=false, Limit Acceleration=false)
   &FN_UpdateAnimTrail() {module=FN_UpdateAnimTrail001}
   &ScaleColor(
    Scale Alpha=&FloatFromCurve(
     FloatCurve=DataInterface() {propertyValues {Curve {keys [
         {arriveTangent=-9.8626, interpMode=RCIM_Cubic, leaveTangent=-9.8626, tangentMode=RCTM_User, time=0, value=2}
         {arriveTangent=-1.1117, interpMode=RCIM_Cubic, leaveTangent=-1.1117, tangentMode=RCTM_User, time=0.5, value=0.5}
         {interpMode=RCIM_Cubic, tangentMode=RCTM_User, time=1, value=0}
        ]}}}
     Scale Curve=&FloatFromCurve(
      FloatCurve=DataInterface() {propertyValues {Curve {keys [<at 0.1 1 RCIM_Cubic>,<at 0.2 0 RCIM_Cubic>]}}}
      CurveIndex=$Emitter.Age
      Scale Curve=&FNTimeOfDayScalar(Night Scale=1.3, Morning Scale=1, Evening Scale=1)
     )
    )
    Scale RGB=&MakeVectorFromLinearColorRGB(Color=$System.ToDColor)
    ScaleA=true
    ScaleRGB=true
   )
   &ScaleRibbonWidth(Ribbon Width Scale=&FloatFromCurve(FloatCurve=DataInterface() {propertyValues {Curve {keys [<at 0 0.5>,<at 0.1 1>,<at 1 0.3>]}}}))
  ]
  rendererClasses [&NiagaraRibbonRendererProperties]
  renderers [{rendererClass=&NiagaraRibbonRendererProperties}]
 }
 TrailSoft {
  emitterSpawnScript [&FN_RibbonIDOffsetUtility()]
  emitterUpdateScript [
   &EmitterState()
   &FN_SpawnAnimTrail()
  ]
  particleSpawnScript [
   &FN_InitializeAnimTrailBoneIndecies()
   &FN_InitializeAnimTrail(Color=<rgba 0.3712 0.1746 0.0319 1>, Lifetime=0.4)
   &AddVelocity(Velocity=<800 0 0>) {enabled=false}
  ]
  particleUpdateScript [
   &SolveForcesAndVelocity(Clamp Velocity=false, Limit Acceleration=false)
   &FN_UpdateAnimTrail() {module=FN_UpdateAnimTrail001}
   &ScaleColor(
    Scale Alpha=&FloatFromCurve(
     FloatCurve=DataInterface() {propertyValues {Curve {keys [<at 0.07 1.3 RCIM_Cubic>,<at 0.17 0 RCIM_Cubic>]}}}
     CurveIndex=$Emitter.Age
     Scale Curve=&FloatFromCurve(
      FloatCurve=DataInterface() {propertyValues {Curve {keys [<at 0.1 1 RCIM_Cubic>, {arriveTangent=-1.9689, interpMode=RCIM_Cubic, leaveTangent=-1.9689, tangentMode=RCTM_User, time=0.6, value=0}]}}}
      CurveIndex=$Emitter.Age
      Scale Curve=&FNTimeOfDayScalar(Night Scale=1.3, Morning Scale=1, Evening Scale=1)
     )
    )
    Scale RGB=&MakeVectorFromLinearColorRGB(Color=$System.ToDColor)
    ScaleA=true
    ScaleRGB=true
   )
  ]
  rendererClasses [&NiagaraRibbonRendererProperties]
  renderers [{rendererClass=&NiagaraRibbonRendererProperties}]
 }
}
summary {
 systemName=NS_Pickaxe_Guitar_CosmicTrellis_SwingTrail
 emitters [
  {emitterName=TrailSharp, rendererClasses [&NiagaraRibbonRendererProperties]}
  {emitterName=TrailSoft, rendererClasses [&NiagaraRibbonRendererProperties]}
 ]
 userVariables [
  User.bIsFrontEnd {type=NiagaraBool}
  User.bIsFrontEndPreview {type=NiagaraBool}
  User.RibbonIDOffset {type=NiagaraInt32}
 ]
}
